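# Row widget for the SoundLib Event Watcher editor utility that displays one audio parameter as a name and value pair.
# FigureOutParam breaks an Audio Parameter struct, stores its name and branches on type to format float, boolean, integer or string values.
# UpdateParam fills both text blocks and tints only the name with a per-type color.
# Object and array parameters match no branch, leaving the displayed text unchanged.
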WidgetBlueprint /Game/Sounds/Fort_Audio_Blueprints/Utilities/EditorUtilityWidgets/SoundLibEventWatcher/ParamWidgetEntry
graphs [
 EventGraph {links [], nodes []}
 FigureOutParam {nodes [
   n1 = FigureOutParam!() -> (then, AudioParameter:Audio Parameter Structure) {is=title K2Node_FunctionEntry}
   n2 = Break Audio Parameter!(AudioParameter:"Audio Parameter Structure (by ref)"=@n1) -> (
    ParamName:Name
    FloatParam:float=~
    BoolParam:bool=~
    IntParam:int
    ObjectParam:Object Reference
    StringParam:String
    ArrayFloatParam:"Array of Float (single-precision)s"
    ArrayBoolParam:Array of Booleans
    ArrayIntParam:Array of Integers
    ArrayObjectParam:Array of Object References
    ArrayStringParam:Array of Strings
    ParamType:EAudioParameterType Enum
   ) {is=title K2Node_BreakStruct}
   > Set Name(Name:Name=@n2) -> (Output_Get:Name=None)
   > n3 = Branch(Condition="Equal (Enum)"(A:EAudioParameterType Enum=@n2.ParamType, B:EAudioParameterType Enum=Float))
   > UpdateParam!(
    execute
    self
    Param:Text="To Text (Name)"!(InName:Name=Get Name() -> (Name:Name=None)) -> (ReturnValue:Text)
    Value:Text="To Text (Float)"(Value=@n2.FloatParam, MinimumIntegralDigits=1, MaximumIntegralDigits=324, MinimumFractionalDigits=2, MaximumFractionalDigits=2)
    Color:Linear Color Structure=(R=0.0395,G=0.6584,B=0,A=1)
   ) -> (then)
   n4 = Branch(execute=@n3.else, Condition="Equal (Enum)"(A=@n2.ParamType, B=Boolean))
   > UpdateParam(
    Param="To Text (Name)"(InName=Get Name() -> (Name=None))
    Value="To Text (String)"(InString="To String (Boolean)"(InBool=@n2.BoolParam))
    Color=(R=0.651,G=0,B=0,A=1)
   )
   n5 = Branch(execute=@n4.else, Condition="Equal (Enum)"(A=@n2.ParamType, B=Integer))
   > UpdateParam(
    Param="To Text (Name)"(InName=Get Name() -> (Name=None))
    Value="To Text (Integer)"(Value=@n2.IntParam)
    Color=(R=0.0137,G=0.7529,B=0.4179,A=1)
   )
   Branch(execute=@n5.else, Condition="Equal (Enum)"(A=@n2.ParamType, B=String))
   > UpdateParam(
    Param="To Text (Name)"(InName=Get Name() -> (Name=None))
    Value="To Text (String)"(InString=@n2.StringParam)
    Color=(R=0.9911,G=0,B=0.6514,A=1)
   )
  ]}
 UpdateParam {nodes [
   n1 = UpdateParam!() -> (then, Param:Text, Value:Text, Color:Linear Color Structure) {is=title K2Node_FunctionEntry}
   > "SetText (Text)"(self=Get Text_ParamName() -> (Text_ParamName:Text Block Object Reference), InText=@n1)
   > "SetText (Text)"(self=Get Text_Value() -> (Text_Value:Text Block Object Reference), InText=@n1.Value)
   > SetColorAndOpacity(self=Get Text_ParamName(), InColorAndOpacity=Make Slate Color(SpecifiedColor=@n1.Color))
  ]}
]
variables [Name {type=string, default=None}]
widgets {
 info {inheritedWidgetCount=0, parentClass=&UserWidget, rootWidgetClass=&CanvasPanel, widgetCount=5}
 widgets [
  CanvasPanel_37 {widgetClassPath=&CanvasPanel, widget=CanvasPanel_37}
  HorizontalBox_49 {
   widgetClassPath=&HorizontalBox
   parent=CanvasPanel_37
   widget=HorizontalBox_49
   slot {
    class=CanvasPanelSlot
    properties {bAutoSize=true, layoutData {alignment=<0 0>, anchors {maximum=<0 0>, minimum=<0 0>}, offsets=<ltrb 0 0 243.6079 30>}}
   }
  }
  Text_ParamName {
   widgetClassPath=&UEFN_TextBlock
   bIsVariable=true
   parent=HorizontalBox_49
   widget=Text_ParamName
   properties {
    text=Parameter Unset
    font {
     bForceMonospaced=false
     bMaterialIsStencil=false
     fontMaterial=null
     fontObject=&NotoSans
     letterSpacing=5
     monospacedWidth=1
     size=12
     skewAmount=0
     typefaceFontName=NotoSans-Regular
     outlineSettings {
      bApplyOutlineToDropShadows=false
      bMiteredCorners=false
      bSeparateFillAlpha=false
      outlineMaterial=null
      outlineSize=0
      outlineColor=<rgba 0 0 0 1>
     }
    }
   }
   slot {class=HorizontalBoxSlot}
  }
  UEFN_TextBlock_C_31 {
   widgetClassPath=&UEFN_TextBlock
   parent=HorizontalBox_49
   widget=UEFN_TextBlock_C_31
   properties {
    text=": "
    font {
     bForceMonospaced=false
     bMaterialIsStencil=false
     fontMaterial=null
     fontObject=&NotoSans
     letterSpacing=0
     monospacedWidth=1
     size=12
     skewAmount=0
     typefaceFontName=NotoSans-Regular
     outlineSettings {
      bApplyOutlineToDropShadows=false
      bMiteredCorners=false
      bSeparateFillAlpha=false
      outlineMaterial=null
      outlineSize=0
      outlineColor=<rgba 0 0 0 1>
     }
    }
   }
   slot {class=HorizontalBoxSlot}
  }
  Text_Value {
   widgetClassPath=&UEFN_TextBlock
   bIsVariable=true
   parent=HorizontalBox_49
   widget=Text_Value
   properties {
    text=1.0
    font {
     bForceMonospaced=false
     bMaterialIsStencil=false
     fontMaterial=null
     fontObject=&NotoSans
     letterSpacing=5
     monospacedWidth=1
     size=12
     skewAmount=0
     typefaceFontName=NotoSans-Regular
     outlineSettings {
      bApplyOutlineToDropShadows=false
      bMiteredCorners=false
      bSeparateFillAlpha=false
      outlineMaterial=null
      outlineSize=0
      outlineColor=<rgba 0 0 0 1>
     }
    }
   }
   slot {class=HorizontalBoxSlot}
  }
 ]
}
