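# ParamWidget is an editor utility sub-widget in SoundLibEventWatcher that lists audio parameters from a Fort pawn's or weapon's Audio Parameter Component.
# Pawn and weapon events cache a valid component, then loop over its parameters in a shared path.
# New names create entries in a vertical dynamic entry box; known names reuse the entry at the matching ParamNames index, and FigureOutParam updates each.
# ClearWidget empties the name array and entry box.
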WidgetBlueprint /Game/Sounds/Fort_Audio_Blueprints/Utilities/EditorUtilityWidgets/SoundLibEventWatcher/ParamWidget
graphs [EventGraph {nodes [
   n1 = GetAudioParamsFromPawn!() -> (OutputDelegate, then, FortPawn:Fort Pawn Object Reference) {is=title K2Node_CustomEvent}
   n2 = GetAudioParameterComponent!(self:Fort Pawn Object Reference=@n1.FortPawn) -> (ReturnValue:Audio Parameter Component Object Reference)
   > Is Valid(InputObject=@n2)
   > n3 = Set AudioParameterComponentRef(AudioParameterComponentRef:Audio Parameter Component Object Reference=@n2) -> (Output_Get:Audio Parameter Component Object Reference)
   n4 = For Each Loop(
    Exec=[@n15.then, @n3.then]
    Array:Array of Audio Parameter Structures=GetParameters!(self:Audio Parameter Component Object Reference=Get AudioParameterComponentRef() -> (AudioParameterComponentRef:Audio Parameter Component Object Reference)) -> (ReturnValue:Array of Audio Parameter Structures)
   ) -> (Array Element:Audio Parameter Structure)
   n5 = Break Audio Parameter!(AudioParameter:"Audio Parameter Structure (by ref)"=@n4) -> (
    ParamName:Name
    FloatParam:float=~
    BoolParam:bool=~
    IntParam:int
    ObjectParam:Object Reference
    StringParam:String
    ArrayFloatParam:"Array of Float (single-precision)s"
    ArrayBoolParam:Array of Booleans
    ArrayIntParam:Array of Integers
    ArrayObjectParam:Array of Object References
    ArrayStringParam:Array of Strings
    ParamType:EAudioParameterType Enum
   ) {is=title K2Node_BreakStruct}
   > n6 = Branch(Condition=Contains Item(TargetArray=Get ParamNames() -> (ParamNames:Array of Names), ItemToFind=@n5))
   n7 = "Get (a copy)"(
    Array:Array of User Widget Object References=GetAllEntries!(self:Dynamic Entry Box Base Object Reference=Get DynamicEntryBox_17() -> (DynamicEntryBox_17:Dynamic Entry Box Object Reference)) -> (ReturnValue:Array of User Widget Object References)
    Dimension 1=Find Item!(TargetArray:Array of Names=Get ParamNames(), ItemToFind:"Name (by ref)"=@n5) -> (ReturnValue)
   ) -> (Output:User Widget Object Reference)
   > Is Valid(InputObject=@n7)
   > n8 = Cast To ParamWidgetEntry!(execute, Object=@n7) -> (then, CastFailed, AsParam Widget Entry:Param Widget Entry Object Reference) {is=title K2Node_DynamicCast}
   > n9 = Set ParamWidgetEntry(ParamWidgetEntry:Param Widget Entry Object Reference=@n8) -> (Output_Get:Param Widget Entry Object Reference)
   FigureOutParam!(
    execute=[@n12.then, @n9.then]
    self:Param Widget Entry Object Reference=Get ParamWidgetEntry() -> (ParamWidgetEntry:Param Widget Entry Object Reference)
    AudioParameter:Audio Parameter Structure=@n4
   ) -> (then)
   Add(execute=@n6.else, TargetArray:Array of Names=Get ParamNames(), NewItem:"Name (by ref)"=@n5) {is=title K2Node_CallArrayFunction}
   > n10 = Create Entry!(execute, self:Dynamic Entry Box Object Reference=Get DynamicEntryBox_17()) -> (then, ReturnValue:User Widget Object Reference)
   > n11 = Cast To ParamWidgetEntry(Object=@n10)
   > n12 = Set ParamWidgetEntry(ParamWidgetEntry=@n11)
   ClearWidget() {is=title K2Node_CustomEvent}
   > Clear(TargetArray:Array of Names=Get ParamNames())
   > Reset!(execute, self:Dynamic Entry Box Object Reference=Get DynamicEntryBox_17(), bDeleteWidgets:bool=true) -> (then)
   n13 = GetAudioParamsFromWeapon!() -> (OutputDelegate, then, FortWeapon:Fort Weapon Object Reference) {is=title K2Node_CustomEvent}
   n14 = GetAudioParameterComponent(self:Fort Weapon Object Reference=@n13.FortWeapon)
   > Is Valid(InputObject=@n14)
   > n15 = Set AudioParameterComponentRef(AudioParameterComponentRef=@n14)
  ]}]
variables [
 AudioParameterComponentRef {type=AudioParameterComponent, default=null}
 ParamNames {type="array<string>", default []}
 ParamWidgetEntry {type=ParamWidgetEntry_C, default=null}
]
widgets {
 info {inheritedWidgetCount=0, parentClass=&UserWidget, rootWidgetClass=&DynamicEntryBox, widgetCount=1}
 widgets [DynamicEntryBox_17 {
   widgetClassPath=&DynamicEntryBox
   bIsVariable=true
   widget=DynamicEntryBox_17
   properties {
    entryBoxType=Vertical
    entryWidgetClass=&/ParamWidgetEntry.ParamWidgetEntry_C
   }
  }]
}
