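# List entry widget for the Sound Library Event Watcher editor utility, showing one sound library event by its gameplay tag.
# When the list item object is set, it casts and stores the entry object, then writes the event tag's debug string into the text block.
# Each Tick maps time since the entry's TimeSinceExecuted timestamp over the watcher's TextFlashFadeOutTime, flashing the text red and fading it back to white.
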WidgetBlueprint /Game/Sounds/Fort_Audio_Blueprints/Utilities/EditorUtilityWidgets/SoundLibEventWatcher/SoundLibraryEventItemWidget
graphs [EventGraph {nodes [
   Event Tick()
   n1 = Get EntryObject() -> (EntryObject:Sound Library Event Entry Object Object Reference)
   n2 = Get EntryObject()
   n3 = MapRangeClamped(
    Value=float - float(
     A=GetTimeSeconds!() -> (ReturnValue:double)
     B=Get TimeSinceExecuted!(self:Sound Library Event Entry Object Object Reference=@n1) -> (TimeSinceExecuted:double)
    )
    InRangeA=0
    InRangeB=Get TextFlashFadeOutTime!(self:EUW Sound Library Event Watcher Object Reference=Get Watcher!(self:Sound Library Event Entry Object Object Reference=@n2) -> (Watcher:EUW Sound Library Event Watcher Object Reference)) -> (TextFlashFadeOutTime:double)
    OutRangeA=1
    OutRangeB=0
   )
   > SetColorAndOpacity(
    self=Get Text!() -> (Text:Text Block Object Reference)
    InColorAndOpacity=Make Slate Color(SpecifiedColor="Lerp (LinearColor)"(A=(R=1,G=1,B=1,A=1), B=(R=1,G=0,B=0,A=1), Alpha=@n3))
   )
   n4 = Event OnListItemObjectSet()
   > n5 = Cast To SoundLibraryEventEntryObject!(execute, Object=@n4.ListItemObject) -> (then, CastFailed, AsSound Library Event Entry Object:Sound Library Event Entry Object Object Reference) {is=title K2Node_DynamicCast}
   > n6 = Set EntryObject(EntryObject:Sound Library Event Entry Object Object Reference=@n5) -> (Output_Get:Sound Library Event Entry Object Object Reference)
   n7 = Get EventTag!(self:Sound Library Event Entry Object Object Reference=@n6) -> (EventTag:Gameplay Tag Structure)
   > "SetText (Text)"(
    self=Get Text()
    InText="To Text (String)"(InString=GetDebugStringFromGameplayTag!(GameplayTag:Gameplay Tag Structure=@n7) -> (ReturnValue:String))
   )
  ]}]
variables [EntryObject {type=SoundLibraryEventEntryObject_C, default=null}]
widgets {
 info {inheritedWidgetCount=0, parentClass=&UserWidget, rootWidgetClass=&CanvasPanel, widgetCount=4}
 widgets [
  CanvasPanel_37 {widgetClassPath=&CanvasPanel, widget=CanvasPanel_37}
  HorizontalBox_49 {
   widgetClassPath=&HorizontalBox
   parent=CanvasPanel_37
   widget=HorizontalBox_49
   slot {class=CanvasPanelSlot, properties {bAutoSize=true, layoutData=~}}
  }
  SizeBox_88 {
   widgetClassPath=&SizeBox
   parent=HorizontalBox_49
   widget=SizeBox_88
   properties {widthOverride=500}
   slot {class=HorizontalBoxSlot}
  }
  Text {
   widgetClassPath=&UEFN_TextBlock
   bIsVariable=true
   parent=SizeBox_88
   widget=Text
   properties {
    text=Event
    font {
     bForceMonospaced=false
     bMaterialIsStencil=false
     fontMaterial=null
     fontObject=&NotoSans
     letterSpacing=5
     monospacedWidth=1
     size=12
     skewAmount=0
     typefaceFontName=NotoSans-Regular
     outlineSettings {
      bApplyOutlineToDropShadows=false
      bMiteredCorners=false
      bSeparateFillAlpha=false
      outlineMaterial=null
      outlineSize=0
      outlineColor=<rgba 0 0 0 1>
     }
    }
   }
   slot {class=SizeBoxSlot, properties {padding=<ltrb 0 0 300 0>}}
  }
 ]
}
