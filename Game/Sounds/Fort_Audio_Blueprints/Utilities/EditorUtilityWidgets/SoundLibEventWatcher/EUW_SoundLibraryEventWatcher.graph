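# Editor utility widget for audio debugging that lists a target actor and its attached actors with their Sound Library play events.
# Tick follows the local player's view target until TargetActor is manually edited, resets on world change, and refreshes pawn/weapon audio params during PIE.
# AddActor skips excluded classes, finds or adds a SoundLibraryComponent, and binds its play event.
# Events feed matching list rows, optionally playing a test tone for expanded ones.
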EditorUtilityWidgetBlueprint /Game/Sounds/Fort_Audio_Blueprints/Utilities/EditorUtilityWidgets/SoundLibEventWatcher/EUW_SoundLibraryEventWatcher
graphs [
 AddActor {nodes [
   n1 = AddActor!() -> (then, New Actor:Actor Object Reference) {is=title K2Node_FunctionEntry}
   > n2 = Branch(Condition=IsActorClassExcluded!(Object:Object Reference=@n1) -> (Result:bool))
   n3 = Branch(
    execute=@n2.else
    Condition=GetActorEntryObject!(Actor:Actor Object Reference=@n1) -> (Result:bool, Entry Object:Sound Library Actor Entry Object Object Reference)
   )
   n4 = FindOrAddSoundLibraryComponent!(execute=@n3.else, self, Actor:Actor Object Reference=@n1) -> (then, ReturnValue:Sound Library Component Object Reference)
   > Bind Event to On Sound Library Play Event(self:Sound Library Component Object Reference=@n4, Delegate=Create Event()) {is=title K2Node_AddDelegate}
   > n5 = Construct Object from Class!(
    execute
    Class=&/SoundLibraryActorEntryObject.SoundLibraryActorEntryObject_C
    self
    Watcher:EUW Sound Library Event Watcher Object Reference=Self-Reference()
   ) -> (then, ReturnValue:Sound Library Actor Entry Object Object Reference)
   > SetActor!(execute, self:Sound Library Actor Entry Object Object Reference=@n5, Actor:Actor Object Reference=@n1) -> (then)
   > AddItem!(execute, self:List View Object Reference=Get ActorList() -> (ActorList:List View Object Reference), Item:Object Reference=@n5) -> (then)
   > Bind Event to On End Play(self:Actor Object Reference=@n1, Delegate=Create Event()) {is=title K2Node_AddDelegate}
  ]}
 EventGraph {nodes [
   Event Tick()
   > n1 = Sequence!(execute) -> (then_0, then_1, then_2:Exec, then_3:Exec, then_4:Exec)
   n2 = Delta Gate Wildcard!(
    execute:Exec=@n1.then_1
    Variable:"Actor Object Reference (by ref)"=Get TargetActorCached() -> (TargetActorCached:Actor Object Reference)
    New Value:Actor Object Reference=Resolve Soft Reference!(Input:Actor Soft Object Reference=Get TargetActor() -> (TargetActor:Actor Soft Object Reference)) -> (Output:Actor Object Reference)
   ) -> (Success:Exec, Updated Variable:"Actor Object Reference (by ref)", Previous Variable:Actor Object Reference) {is=title K2Node_MacroInstance}
   > RemoveAllActors()
   > AddActor!(execute, self, New Actor:Actor Object Reference=@n2) -> (then) {is=title K2Node_CallFunction}
   n3 = GetPIEWorlds!(execute=@n1.then_4, bIncludeDedicatedServer:bool=false) -> (then, ReturnValue:Array of World Object References)
   > Branch(Condition=Is Not Empty!(TargetArray:Array of World Object References=@n3) -> (ReturnValue:bool))
   > n4 = Is Valid(InputObject="Get (a copy)"(Array:Array of World Object References=@n3) -> (Output:World Object Reference))
   n5 = Set ClearParamWidget(execute=@n4.Is Not Valid, ClearParamWidget:bool=true) -> (Output_Get:bool)
   n6 = For Each Loop(
    Exec=[@n15.then, @n5.then]
    Array:Array of Object References=GetListItems!(self:List View Object Reference=Get ActorList()) -> (ReturnValue:Array of Object References)
   ) -> (Array Element:Object Reference)
   > n7 = Cast To SoundLibraryActorEntryObject!(execute, Object=@n6) -> (then, CastFailed, AsSound Library Actor Entry Object:Sound Library Actor Entry Object Object Reference) {is=title K2Node_DynamicCast}
   n8 = Get ActorList()
   > n9 = Cast To SoundLibraryActorItemWidget!(
    execute
    Object="Get (a copy)"(
     Array:Array of User Widget Object References=GetDisplayedEntryWidgets!(self:List View Base Object Reference=@n8) -> (ReturnValue:Array of User Widget Object References)
     Dimension 1=@n6.Array Index
    ) -> (Output:User Widget Object Reference)
   ) -> (then, CastFailed, AsSound Library Actor Item Widget:Sound Library Actor Item Widget Object Reference) {is=title K2Node_DynamicCast}
   > n10 = Branch(Condition=Get ClearParamWidget() -> (ClearParamWidget:bool))
   n11 = Get ParamWidget!(self:Sound Library Actor Item Widget Object Reference=@n9) -> (ParamWidget:Param Widget Object Reference)
   > ClearWidget(self:Param Widget Object Reference=@n11)
   n12 = Get Actor!(self:Sound Library Actor Entry Object Object Reference=@n7) -> (Actor:Actor Object Reference)
   n13 = Cast To FortPawn!(execute=@n10.else, Object=@n12) -> (then, CastFailed, AsFort Pawn:Fort Pawn Object Reference) {is=title K2Node_DynamicCast}
   > GetAudioParamsFromPawn!(execute, self:Param Widget Object Reference=@n11, FortPawn:Fort Pawn Object Reference=@n13) -> (then)
   n14 = Cast To FortWeapon!(execute=@n13.CastFailed, Object=@n12) -> (then, CastFailed, AsFort Weapon:Fort Weapon Object Reference) {is=title K2Node_DynamicCast}
   > GetAudioParamsFromWeapon!(execute, self:Param Widget Object Reference=@n11, FortWeapon:Fort Weapon Object Reference=@n14) -> (then)
   n15 = Set ClearParamWidget(execute=@n4.Is Valid, ClearParamWidget=false)
   Delta Gate Wildcard(
    execute=@n1.then_0
    Variable:"Object Reference (by ref)"=Get CurrentWorld() -> (CurrentWorld:Object Reference)
    New Value:Object Reference=GetCurrentWorld!() -> (ReturnValue:World Object Reference)
   ) -> (Updated Variable:"Object Reference (by ref)", Previous Variable:Object Reference)
   > RemoveActor!(execute, self, Actor:Actor Object Reference=Get TargetActorCached()) -> (then)
   > Set TargetActor(TargetActor:Actor Soft Object Reference) -> (Output_Get:Actor Soft Object Reference)
   n16 = Get!(execute:Exec=@n1.then_3) -> (then:Exec, else:Exec, TargetActorCached:Actor Object Reference)
   > n17 = For Each Loop(Array:Array of Actor Object References=GetAttachedActors!(self:Actor Object Reference=@n16, bResetArray:bool=true, bRecursivelyIncludeAttachedActors:bool=false) -> (OutActors:Array of Actor Object References)) -> (Array Element:Actor Object Reference)
   > AddActor(New Actor=@n17)
   Branch(execute=@n1.then_2, Condition=Get bAutoWatchLocalPlayer() -> (bAutoWatchLocalPlayer:bool))
   n18 = GetFirstLocalFortPlayerController!(WorldContextObject:Object Reference=Get CurrentWorld(), bIncludeNonSpectators:bool=true, bIncludeSpectators:bool=false) -> (ReturnValue:Fort Player Controller Object Reference)
   > Is Valid(InputObject=@n18)
   n19 = GetViewTarget!(self:Controller Object Reference=@n18) -> (ReturnValue:Actor Object Reference)
   > Is Valid(InputObject=@n19)
   > n20 = Cast To PlayerController!(execute, Object=@n19) -> (then, CastFailed, AsPlayer Controller:Player Controller Object Reference) {is=title K2Node_DynamicCast}
   Set TargetActor(execute=@n20.CastFailed, TargetActor=Make Soft Reference!(Input:Actor Object Reference=@n19) -> (Output:Actor Soft Object Reference))
   Event Construct()
   n21 = Get FortMultiPropertyView() -> (FortMultiPropertyView:Fort Multi Property View Object Reference)
   > SetObject(self=@n21, NewObject=Self-Reference())
   > Bind Event to On Property Changed(self:Property View Base Object Reference=@n21, Delegate=Create Event()) {is=title K2Node_AddDelegate}
  ]}
 FindOrAddSoundLibraryComponent {nodes [
   n1 = FindOrAddSoundLibraryComponent!() -> (then, Actor:Actor Object Reference) {is=title K2Node_FunctionEntry}
   n2 = GetComponentByClass!(self:Actor Object Reference=@n1, ComponentClass:Actor Component Class Reference=/Script/SoundLibrary.SoundLibraryComponent) -> (ReturnValue:Sound Library Component Object Reference)
   > n3 = Is Valid(InputObject=@n2)
   n4 = Add Component by Class!(
    execute:Exec=@n3.Is Not Valid
    self:Actor Object Reference=@n1
    WorldContextObject:Object Reference=@n1
    Class:Actor Component Class Reference=&SoundLibraryComponent
   ) -> (then:Exec, ReturnValue:Sound Library Component Object Reference) {is=title K2Node_AddComponentByClass}
   > Return Node!(execute, ReturnValue:Sound Library Component Object Reference=@n4)
   Return Node(execute=@n3.Is Valid, ReturnValue=@n2)
  ]}
 GetActorEntryObject {nodes [
   n1 = GetActorEntryObject!() -> (then, Actor:Actor Object Reference) {is=title K2Node_FunctionEntry}
   > n2 = For Each Loop(Array:Array of Object References=GetListItems(self=Get ActorList())) -> (Array Element:Object Reference)
   Return Node!(execute=[@n3.CastFailed, @n2.Completed], Result:bool=false, Entry Object:Sound Library Actor Entry Object Object Reference)
   > n3 = Cast To SoundLibraryActorEntryObject(Object=@n2)
   > Branch(Condition="Equal (Object)"(A=Get Actor(self=@n3), B=@n1))
   > Return Node(Result=true, Entry Object=@n3)
  ]}
 GetCurrentWorld {nodes [
   GetCurrentWorld() {is=title K2Node_FunctionEntry}
   > n1 = GetPIEWorlds(bIncludeDedicatedServer=false)
   > n2 = Branch(Condition=Is Valid Index!(TargetArray:Array of World Object References=@n1, IndexToTest:int) -> (ReturnValue:bool))
   > Return Node!(execute, ReturnValue:World Object Reference="Get (a copy)"(Array:Array of World Object References=@n1) -> (Output:World Object Reference))
   n3 = GetEditorWorld(execute=@n2.else, self=Get UnrealEditorSubsystem())
   > Return Node(ReturnValue=@n3)
  ]}
 IsActorClassExcluded {nodes [
   n1 = IsActorClassExcluded!() -> (then, Object:Object Reference) {is=title K2Node_FunctionEntry}
   > n2 = For Each Loop(Array:Array of Actor Class References=Get ClassesToExclude() -> (ClassesToExclude:Array of Actor Class References)) -> (Array Element:Actor Class Reference)
   Return Node!(execute=@n2.Completed, Result:bool=false)
   > Branch(Condition=ClassIsChildOf(TestClass=Get Class(Object=@n1) -> (ReturnValue:Object Class Reference), ParentClass=@n2))
   > Return Node(Result=true)
  ]}
 OnActorEndPlay {nodes [
   n1 = OnActorEndPlay!() -> (then, Actor:Actor Object Reference, EndPlayReason:EEndPlayReason Enum=Destroyed) {is=title K2Node_FunctionEntry}
   > RemoveActor(Actor=@n1)
   > Unbind Event from On End Play(self:Actor Object Reference=@n1, Delegate=Create Event()) {is=title K2Node_RemoveDelegate}
   n2 = Get TargetActor()
   > Branch(Condition="Equal (Object)"(A=Resolve Soft Reference(Input=@n2), B=@n1))
   > Set TargetActor()
  ]}
 OnSettingsPropertyChanged {nodes [
   n1 = OnSettingsPropertyChanged!() -> (then, PropertyName:Name) {is=title K2Node_FunctionEntry}
   > n2 = Switch on Name!(execute:Exec, Selection:Name=@n1) -> (Default:Exec, TargetActor:Exec) {is=title K2Node_SwitchName, pinNames [TargetActor]}
   Set bAutoWatchLocalPlayer(execute=@n2.TargetActor, bAutoWatchLocalPlayer:bool=false) -> (Output_Get:bool)
  ]}
 OnSoundLibraryPlayEvent {nodes [
   n1 = OnSoundLibraryPlayEvent!() -> (then, Results:"Sound Library Play Result Structure (by ref)", Context:Sound Library Context Object Reference) {is=title K2Node_FunctionEntry}
   > n2 = For Each Loop(Array:Array of User Widget Object References=GetDisplayedEntryWidgets(self=Get ActorList())) -> (Array Element:User Widget Object Reference)
   > n3 = Cast To SoundLibraryActorItemWidget(Object=@n2)
   n4 = Break Sound Library Play Result!(SoundLibraryPlayResult:"Sound Library Play Result Structure (by ref)"=@n1) -> (EventName:Gameplay Tag Structure, bWasBlocked:bool=~, bSuccess:bool=~, AudioComponents:Array of Audio Component Object References) {is=title K2Node_BreakStruct}
   n5 = Get Actor(self:Sound Library Context Object Reference=Get Context() -> (Context:Sound Library Context Object Reference)) -> (Actor:Actor Weak object)
   n6 = Get Actor(self:Sound Library Actor Entry Object Object Reference=Get EntryObject!(self:Sound Library Actor Item Widget Object Reference=@n3) -> (EntryObject:Sound Library Actor Entry Object Object Reference)) -> (Actor:Actor Object Reference)
   > Branch(Condition=AND Boolean(A="Equal (Object)"(A=@n6, B=@n5), B=IsGameplayTagValid!(GameplayTag:Gameplay Tag Structure=@n4) -> (ReturnValue:bool)))
   > AddEvent!(execute, self:Sound Library Actor Item Widget Object Reference=@n3, EventData:"Sound Library Play Result Structure (by ref)"=@n1) -> (then)
   n7 = Get bIsCollapsed!(self:Sound Library Actor Item Widget Object Reference=@n3) -> (bIsCollapsed:bool)
   > Branch(Condition=AND Boolean(A=NOT Boolean(A=@n7), B=Get bPlayTestTone() -> (bPlayTestTone:bool)))
   PlaySound2D!(
    execute
    WorldContextObject:Object Reference=Get Actor(self:Sound Library Context Object Reference=Get Context()) -> (Actor:Actor Weak object)
    Sound=/Game/Sounds/Fort_Test_Sounds/Tones/1k_test_tone
    VolumeMultiplier=0.1
    PitchMultiplier
    StartTime
    ConcurrencySettings
    OwningActor
    bIsUISound
   ) -> (then)
   Return Node!(execute)
  ]}
 RemoveActor {nodes [
   n1 = RemoveActor!() -> (then, Actor:Actor Object Reference) {is=title K2Node_FunctionEntry}
   n2 = GetActorEntryObject(Actor=@n1) {is=title K2Node_CallFunction}
   > Branch(Condition=@n2)
   > n3 = FindOrAddSoundLibraryComponent(Actor=@n1) {is=title K2Node_CallFunction}
   > n4 = Is Valid(InputObject=@n3)
   > n5 = Unbind Event from On Sound Library Play Event(self:Sound Library Component Object Reference=@n3, Delegate=Create Event()) {is=title K2Node_RemoveDelegate}
   RemoveItem!(execute=[@n5.then, @n4.Is Not Valid], self:List View Object Reference=Get ActorList(), Item:Object Reference=@n2.Entry Object) -> (then)
   > n6 = For Each Loop(Array:Array of Actor Object References=GetAttachedActors(self=@n1, bResetArray=true, bRecursivelyIncludeAttachedActors=false)) -> (Array Element:Actor Object Reference)
   > RemoveActor(Actor=@n6) {is=title K2Node_CallFunction}
  ]}
 RemoveAllActors {nodes [
   RemoveAllActors() {is=title K2Node_FunctionEntry}
   > n1 = For Each Loop(Array:Array of Object References=GetListItems(self=Get ActorList())) -> (Array Element:Object Reference)
   > n2 = Cast To SoundLibraryActorEntryObject(Object=@n1)
   > RemoveActor(Actor=Get Actor(self:Sound Library Actor Entry Object Object Reference=@n2) -> (Actor:Actor Object Reference))
  ]}
]
variables [
 bAutoWatchLocalPlayer {type=boolean, default=true}
 bPlayTestTone {type=boolean}
 ClassesToExclude {
  type="array<Class>"
  default [/Script/Engine.WorldSettings, /Script/GameplayAbilities.GameplayCueNotify_Looping, /Script/GameplayAbilities.GameplayCueNotify_BurstLatent]
 }
 ClearParamWidget {type=boolean}
 CurrentWorld {type=Object, default=null}
 SoundLibraryComponent {type=SoundLibraryComponent, default=null}
 TargetActor {type=Actor, default=null}
 TargetActorCached {type=Actor, default=null}
 TextFlashFadeOutTime {type=number, default=1}
]
widgets {
 info {rootWidgetClass=&CanvasPanel, widgetCount=7}
 widgets [
  CanvasPanel_0 {widgetClassPath=&CanvasPanel, widget=CanvasPanel_0}
  VerticalBox_44 {
   widgetClassPath=&VerticalBox
   parent=CanvasPanel_0
   widget=VerticalBox_44
   slot {
    class=CanvasPanelSlot
    properties {bAutoSize=true, layoutData {alignment=<0 0>, anchors {maximum=<0 0>, minimum=<0 0>}, offsets=<ltrb 10 10 100 30>}}
   }
  }
  SettingsText {
   widgetClassPath=&UEFN_TextBlock
   bIsVariable=true
   parent=VerticalBox_44
   widget=SettingsText
   properties {
    text=Settings
    font {
     bForceMonospaced=false
     bMaterialIsStencil=false
     fontMaterial=null
     fontObject=&NotoSans
     letterSpacing=5
     monospacedWidth=1
     size=12
     skewAmount=0
     typefaceFontName=NotoSans-Regular
     outlineSettings {
      bApplyOutlineToDropShadows=false
      bMiteredCorners=false
      bSeparateFillAlpha=false
      outlineMaterial=null
      outlineSize=0
      outlineColor=<rgba 0 0 0 1>
     }
    }
   }
   slot {class=VerticalBoxSlot}
  }
  Border_321 {
   widgetClassPath=&Border
   parent=VerticalBox_44
   widget=Border_321
   properties {brushColor=<rgba 0 0 0 0.1>, padding=<ltrb 5 5 5 5>}
   slot {class=VerticalBoxSlot}
  }
  FortMultiPropertyView {
   widgetClassPath=/Script/FortniteEditor.FortMultiPropertyView
   bIsVariable=true
   parent=Border_321
   widget=FortMultiPropertyView
   properties {
    propertiesToShow [TargetActor, bAutoWatchLocalPlayer, bPlayTestTone, TextFlashFadeOutTime]
    propertyGridLayout {bAutoWatchLocalPlayer=<0 0>, bPlayTestTone=<0 1>, targetActor=<1 0>, textFlashFadeOutTime=<1 1>}
   }
   slot {class=BorderSlot, properties {padding=<ltrb 5 5 5 5>}}
  }
  EventWatcherText {
   widgetClassPath=&UEFN_TextBlock
   bIsVariable=true
   parent=VerticalBox_44
   widget=EventWatcherText
   properties {
    text=Event Watcher
    font {
     bForceMonospaced=false
     bMaterialIsStencil=false
     fontMaterial=null
     fontObject=&NotoSans
     letterSpacing=5
     monospacedWidth=1
     size=12
     skewAmount=0
     typefaceFontName=NotoSans-Regular
     outlineSettings {
      bApplyOutlineToDropShadows=false
      bMiteredCorners=false
      bSeparateFillAlpha=false
      outlineMaterial=null
      outlineSize=0
      outlineColor=<rgba 0 0 0 1>
     }
    }
   }
   slot {class=VerticalBoxSlot}
  }
  ActorList {
   widgetClassPath=/Script/UMG.ListView
   bIsVariable=true
   parent=VerticalBox_44
   widget=ActorList
   properties {
    entryWidgetClass=&/SoundLibraryActorItemWidget.SoundLibraryActorItemWidget_C
    selectionMode=Multi
    scrollBarStyle {
     thickness=8
     draggedThumbImage {
      bIsDynamicallyLoaded=false
      drawAs=RoundedBox
      imageType=NoImage
      mirroring=NoMirror
      resourceName=None
      resourceObject=null
      tiling=NoTile
      imageSize=<0 0>
      margin=<ltrb 0 0 0 0>
      outlineSettings {
       bUseBrushTransparency=false
       roundingType=FixedRadius
       width=0
       color {colorUseRule=UseColor_Specified, specifiedColor=<rgba 0 0 0 0>}
       cornerRadii=<4 4 4 4>
      }
      tintColor {colorUseRule=UseColor_Specified, specifiedColor=<rgba 0.2159 0.2159 0.2159 1>}
      uVRegion {bIsValid=false, max=<0 0>, min=<0 0>}
     }
     horizontalBackgroundImage {
      bIsDynamicallyLoaded=false
      drawAs=NoDrawType
      imageType=NoImage
      mirroring=NoMirror
      resourceName=None
      resourceObject=null
      tiling=NoTile
      imageSize=<0 0>
      margin=<ltrb 0 0 0 0>
      outlineSettings {
       bUseBrushTransparency=false
       roundingType=HalfHeightRadius
       width=0
       color {colorUseRule=UseColor_Specified, specifiedColor=<rgba 0 0 0 0>}
       cornerRadii=<0 0 0 0>
      }
      tintColor {colorUseRule=UseColor_Specified, specifiedColor=<rgba 1 1 1 1>}
      uVRegion {bIsValid=false, max=<0 0>, min=<0 0>}
     }
     horizontalBottomSlotImage {
      bIsDynamicallyLoaded=false
      drawAs=NoDrawType
      imageType=NoImage
      mirroring=NoMirror
      resourceName=None
      resourceObject=null
      tiling=NoTile
      imageSize=<0 0>
      margin=<ltrb 0 0 0 0>
      outlineSettings {
       bUseBrushTransparency=false
       roundingType=HalfHeightRadius
       width=0
       color {colorUseRule=UseColor_Specified, specifiedColor=<rgba 0 0 0 0>}
       cornerRadii=<0 0 0 0>
      }
      tintColor {colorUseRule=UseColor_Specified, specifiedColor=<rgba 1 1 1 1>}
      uVRegion {bIsValid=false, max=<0 0>, min=<0 0>}
     }
     horizontalTopSlotImage {
      bIsDynamicallyLoaded=false
      drawAs=NoDrawType
      imageType=NoImage
      mirroring=NoMirror
      resourceName=None
      resourceObject=null
      tiling=NoTile
      imageSize=<0 0>
      margin=<ltrb 0 0 0 0>
      outlineSettings {
       bUseBrushTransparency=false
       roundingType=HalfHeightRadius
       width=0
       color {colorUseRule=UseColor_Specified, specifiedColor=<rgba 0 0 0 0>}
       cornerRadii=<0 0 0 0>
      }
      tintColor {colorUseRule=UseColor_Specified, specifiedColor=<rgba 1 1 1 1>}
      uVRegion {bIsValid=false, max=<0 0>, min=<0 0>}
     }
     hoveredThumbImage {
      bIsDynamicallyLoaded=false
      drawAs=RoundedBox
      imageType=NoImage
      mirroring=NoMirror
      resourceName=None
      resourceObject=null
      tiling=NoTile
      imageSize=<0 0>
      margin=<ltrb 0 0 0 0>
      outlineSettings {
       bUseBrushTransparency=false
       roundingType=FixedRadius
       width=0
       color {colorUseRule=UseColor_Specified, specifiedColor=<rgba 0 0 0 0>}
       cornerRadii=<4 4 4 4>
      }
      tintColor {colorUseRule=UseColor_Specified, specifiedColor=<rgba 0.2159 0.2159 0.2159 1>}
      uVRegion {bIsValid=false, max=<0 0>, min=<0 0>}
     }
     normalThumbImage {
      bIsDynamicallyLoaded=false
      drawAs=RoundedBox
      imageType=NoImage
      mirroring=NoMirror
      resourceName=None
      resourceObject=null
      tiling=NoTile
      imageSize=<0 0>
      margin=<ltrb 0 0 0 0>
      outlineSettings {
       bUseBrushTransparency=false
       roundingType=FixedRadius
       width=0
       color {colorUseRule=UseColor_Specified, specifiedColor=<rgba 0 0 0 0>}
       cornerRadii=<4 4 4 4>
      }
      tintColor {colorUseRule=UseColor_Specified, specifiedColor=<rgba 0.0953 0.0953 0.0953 1>}
      uVRegion {bIsValid=false, max=<0 0>, min=<0 0>}
     }
     verticalBackgroundImage {
      bIsDynamicallyLoaded=false
      drawAs=NoDrawType
      imageType=NoImage
      mirroring=NoMirror
      resourceName=None
      resourceObject=null
      tiling=NoTile
      imageSize=<0 0>
      margin=<ltrb 0 0 0 0>
      outlineSettings {
       bUseBrushTransparency=false
       roundingType=HalfHeightRadius
       width=0
       color {colorUseRule=UseColor_Specified, specifiedColor=<rgba 0 0 0 0>}
       cornerRadii=<0 0 0 0>
      }
      tintColor {colorUseRule=UseColor_Specified, specifiedColor=<rgba 1 1 1 1>}
      uVRegion {bIsValid=false, max=<0 0>, min=<0 0>}
     }
     verticalBottomSlotImage {
      bIsDynamicallyLoaded=false
      drawAs=NoDrawType
      imageType=NoImage
      mirroring=NoMirror
      resourceName=None
      resourceObject=null
      tiling=NoTile
      imageSize=<0 0>
      margin=<ltrb 0 0 0 0>
      outlineSettings {
       bUseBrushTransparency=false
       roundingType=HalfHeightRadius
       width=0
       color {colorUseRule=UseColor_Specified, specifiedColor=<rgba 0 0 0 0>}
       cornerRadii=<0 0 0 0>
      }
      tintColor {colorUseRule=UseColor_Specified, specifiedColor=<rgba 1 1 1 1>}
      uVRegion {bIsValid=false, max=<0 0>, min=<0 0>}
     }
     verticalTopSlotImage {
      bIsDynamicallyLoaded=false
      drawAs=NoDrawType
      imageType=NoImage
      mirroring=NoMirror
      resourceName=None
      resourceObject=null
      tiling=NoTile
      imageSize=<0 0>
      margin=<ltrb 0 0 0 0>
      outlineSettings {
       bUseBrushTransparency=false
       roundingType=HalfHeightRadius
       width=0
       color {colorUseRule=UseColor_Specified, specifiedColor=<rgba 0 0 0 0>}
       cornerRadii=<0 0 0 0>
      }
      tintColor {colorUseRule=UseColor_Specified, specifiedColor=<rgba 1 1 1 1>}
      uVRegion {bIsValid=false, max=<0 0>, min=<0 0>}
     }
    }
   }
   slot {class=VerticalBoxSlot}
  }
 ]
}
