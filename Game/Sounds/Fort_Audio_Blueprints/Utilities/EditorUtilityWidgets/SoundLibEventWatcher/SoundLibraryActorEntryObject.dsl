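; Editor utility object for the Sound Library Event Watcher widget that represents one actor entry in its list.
; Its SetActor function takes an actor reference and stores it in the object's Actor variable.
; The file holds no other logic, so it serves as a simple data wrapper around that actor.
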(fn SetActor
  (params (Actor :type Actor%20Object%20Reference))
  (Variables|Default|SetActor Actor))
