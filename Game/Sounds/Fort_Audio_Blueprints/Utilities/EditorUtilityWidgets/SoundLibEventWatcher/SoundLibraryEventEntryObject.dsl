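; SoundLibraryEventEntryObject is a data object for one event entry in the SoundLibEventWatcher editor utility widget for sound library events.
; Its SetEventData function stores a gameplay tag and a time-since-executed value in the object's EventTag and TimeSinceExecuted variables.
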(fn SetEventData
  (params (EventTag :type Gameplay%20Tag%20Structure) (TimeSinceExecuted :type Float%20%28double-precision%29))
  (Variables|Default|SetEventTag EventTag)
  (Variables|Default|SetTimeSinceExecuted TimeSinceExecuted))
